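# Unlit translucent master material for Niagara lens effects, drawn without depth test and transformed to screen space by a particle-to-screen WPO function.
# Two textures with optional radial or swizzled UVs, sampled using unmodified-UV derivatives, are biased, multiplied and shaped by a radial sphere mask.
# The result, optionally recolored through a curve-atlas LUT, drives particle-tinted emissive and eroded alpha.
# Dynamic parameters animate texture scale, offset, bias, mask radius and erosion.
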Material /DealershipSharedLibrary/FX/Materials/MasterMaterials/Lens/M_DLR_LensEffect
graph {
 properties {
  bAutomaticallySetUsageInEditor=false
  bDisableDepthTest=true
  bUseTranslucencyVertexFog=false
  bUsedWithNiagaraMeshParticles=true
  bUsedWithNiagaraSprites=true
  bUsedWithStaticMesh=false
  blendMode=BLEND_Translucent
  pixelDepthOffsetMode=PDOM_Legacy
  shadingModel=MSM_Unlit
  usedShadingModels=Unlit
 }
 nodes [
  n1 = ScalarParameter() {group=Texture 1, parameterName=Emissive - Use LUT}
  n2 = TextureCoordinate()
  n3 = MaterialFunctionCall!(Vector or UVs=@n2) {materialFunction=/FortUILibrary/MaterialFunctions/Basics/MF_UI_CheaperVectorToRadial}
  n4 = If(
   A=ScalarParameter() {group=Texture 1, parameterName=Texture 1 - Use Radial UV}
   "A > B"=@n3.Radial Coordinates
   "A == B"=@n3.Radial Coordinates
   "A < B"=@n2
  ) {constB=1}
  n5 = MaterialFunctionCall(XY=@n4) {materialFunction=&Swizzle}
  n6 = If(A=ScalarParameter() {group=Texture 1, parameterName=Texture 1 - Swizze UV}, "A > B"=@n5.YX, "A == B"=@n5.YX, "A < B"=@n4) {constB=1}
  n7 = NamedRerouteUsage()
  n8 = TextureCoordinate()
  n9 = TextureSampleParameter2D!(UVs=Add(A=Multiply(A=@n6, B=@n7), B=NamedRerouteUsage()), "DDX(UVs)"=DDX(Value=@n8), "DDY(UVs)"=DDY(Value=@n8), Apply View MipBias) {
   group=Texture 1
   mipValueMode=TMVM_Derivative
   parameterName=Texture 1
   samplerType=SAMPLERTYPE_Masks
   texture=/Game/Global/Textures/Default/Blanks/T_White
  }
  n10 = If(
   A=ScalarParameter() {group=Texture 1, parameterName=Texture 1 - Use RGB}
   "A > B"=@n9.RGB
   "A == B"=@n9.RGB
   "A < B"=ChannelMaskParameter(None=@n9.RGBA) {group=Texture 1, parameterName=Texture 1 - Channel}.R
  ) {constB=1}
  n11 = NamedRerouteUsage()
  n12 = TextureCoordinate()
  n13 = MaterialFunctionCall(Vector or UVs=@n12) {materialFunction=&MF_UI_CheaperVectorToRadial}
  n14 = If(
   A=ScalarParameter() {group=Texture 2, parameterName=Texture 2 - Use Radial UV}
   "A > B"=@n13.Radial Coordinates
   "A == B"=@n13.Radial Coordinates
   "A < B"=@n12
  ) {constB=1}
  n15 = MaterialFunctionCall(XY=@n14) {materialFunction=&Swizzle}
  n16 = If(A=ScalarParameter() {group=Texture 2, parameterName=Texture 2 - Swizze UV}, "A > B"=@n15.YX, "A == B"=@n15.YX, "A < B"=@n14) {constB=1}
  n17 = NamedRerouteUsage()
  n18 = TextureCoordinate()
  n19 = TextureSampleParameter2D(UVs=Add(A=Multiply(A=@n16, B=@n17), B=NamedRerouteUsage()), "DDX(UVs)"=DDX(Value=@n18), "DDY(UVs)"=DDY(Value=@n18)) {
   group=Texture 2
   mipValueMode=TMVM_Derivative
   parameterName=Texture 2
   samplerType=SAMPLERTYPE_Masks
   texture=&T_White
  }
  n20 = If(
   A=ScalarParameter() {group=Texture 2, parameterName=Texture 2- Use RGB}
   "A > B"=@n19.RGB
   "A == B"=@n19.RGB
   "A < B"=ChannelMaskParameter(None=@n19.RGBA) {group=Texture 2, parameterName=Texture 2 - Channel}.R
  ) {constB=1}
  n21 = NamedRerouteUsage()
  n22 = ScalarParameter() {defaultValue=1, group=Radial Mask, parameterName=Radial Mask - Min}
  n23 = ScalarParameter() {defaultValue=1, group=Radial Mask, parameterName=Radial Mask - Max}
  n24 = SphereMask(
   A=TextureCoordinate()
   B=Constant() {r=0.5}
   Radius=NamedRerouteUsage()
   Hardness=ScalarParameter() {group=Radial Mask, parameterName=Radial Mask - Hardness}
  ) {attenuationRadius=0.71}
  n25 = Multiply(A=@n24) {constB=0.5}
  n26 = Power(
   Base=LinearInterpolate(A=@n24, B=Sine(None=@n25), Alpha=ScalarParameter() {group=Radial Mask, parameterName=Radial Mask - Use Sine})
   Exp=ScalarParameter() {defaultValue=1, group=Radial Mask, parameterName=Radial Mask - Exp}
  ) {constExponent=1}
  n27 = Multiply(A=Multiply(A=Add(A=@n10, B=@n11), B=Add(A=@n20, B=@n21)), B=Saturate(None=LinearInterpolate(A=@n22, B=@n23, Alpha=@n26)))
  n28 = TextureSampleParameter2D!(UVs=Desaturation(None=@n27), Apply View MipBias) {
   group=Emissive
   parameterName=Emissive - LUT
   samplerType=SAMPLERTYPE_LinearColor
   texture=/DealershipSharedLibrary/FX/Textures/Curves/Curve_Default_Atlas
  }
  MaterialFunctionCall(Emissive=Multiply(A=ParticleColor().RGB, B=If(A=@n1, "A > B"=@n28.RGB, "A == B"=@n28.RGB, "A < B"=@n27) {constB=1})) {materialFunction=&MF_FX_EmissiveAdjustments}
  MaterialFunctionCall!(
   ScreenSizePercentage=VectorParameter() {group=WPO, parameterName=WPO - Screen Size Percentage, defaultValue=<rgba 1 1 1 0>}.RGB
   FitScreenPercentage=ScalarParameter() {defaultValue=1, group=WPO, parameterName=WPO - Fit Screen}
   Align Offset to Particle=ScalarParameter() {group=WPO, parameterName=WPO - Align Camera Offset to Particle}
   CameraOffset=ScalarParameter() {defaultValue=100, group=WPO, parameterName=WPO - Camera Offset}
   Is Sprite=ScalarParameter() {group=WPO, parameterName=WPO - Is Sprite}
  ) {materialFunction=/DealershipSharedLibrary/FX/Materials/MaterialFunctions/MF_DLR_ParticleToScreenTransform}
  n29 = NamedRerouteUsage()
  Multiply(
   A=ParticleColor().A
   B=Multiply(
    A=@n27
    B=MaterialFunctionCall!(Dynamic Param=@n29, Alpha=@n27) {materialFunction=/Game/Effects/Fort_Effects/Materials/_Functions/Erode_Alpha}.Result
   )
  )
  n30 = DynamicParameter() {parameterIndex=2, defaultValue=<rgba 1 1 0 0>, paramNames [T2 - Scale X, T2 - Scale Y, T2 - Offset X, T2 - Offset Y]}
  NamedRerouteDeclaration(None=ComponentMask(None=@n30.RGBA) {g=true, r=true}) {name=Texture 2 - Scale, nodeColor=<rgba 0.6172 0 1 1>}
  n31 = DynamicParameter() {parameterIndex=1, defaultValue=<rgba 1 1 0 0>, paramNames [T1 - Scale X, T1 - Scale Y, T1 - Offset X, T1 - Offset Y]}
  NamedRerouteDeclaration(None=ComponentMask(None=@n31.RGB) {g=true, r=true}) {name=Texture 1 - Scale, nodeColor=<rgba 0.6172 0 1 1>}
  NamedRerouteDeclaration(None=ComponentMask(None=@n31.RGBA) {a=true, b=true}) {name=Texture 1 - Offset, nodeColor=<rgba 0.6172 0 1 1>}
  NamedRerouteDeclaration(None=ComponentMask(None=@n30.RGBA) {a=true, b=true}) {name=Texture 2 - Offset, nodeColor=<rgba 0.6172 0 1 1>}
  n32 = DynamicParameter() {defaultValue=<rgba 1 1 0.71 1>, paramNames [Texture 1 Bias, Texture 2 Bias, Mask Radius, Erode]}
  NamedRerouteDeclaration(None=@n32.Erode) {name=Erode, nodeColor=<rgba 0.6172 0 1 1>}
  NamedRerouteDeclaration(None=@n32.Texture 1 Bias) {name=Texture 1 Bias, nodeColor=<rgba 0.6172 0 1 1>}
  NamedRerouteDeclaration(None=@n32.Texture 2 Bias) {name=Texture 2 Bias, nodeColor=<rgba 0.6172 0 1 1>}
  NamedRerouteDeclaration(None=@n32.Mask Radius) {name=Radial Mask Radius, nodeColor=<rgba 0.6172 0 1 1>}
 ]
}
